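# Unlit material for the advection step of a 2D fluid simulation, shared by velocity and other fields through an IsVelocity switch.
# It samples Source semi-Lagrangian style, backtraced along Velocity by TimeStep over Resolution.
# Velocity gets buoyancy, radial mouse force with curl noise, and dampening; other fields get radial paint and obstacle-mask attenuation.
# MacCormack-style correction and midpoint Runge-Kutta branches are present but feed nothing downstream.
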Material /Volumetrics/Tools/FluidSimulation/Materials/2D/M_Advect
graph {
 properties {bAllowNegativeEmissiveColor=true, pixelDepthOffsetMode=PDOM_Legacy, shadingModel=MSM_Unlit, usedShadingModels=Unlit}
 nodes [
  n1 = TextureSampleParameter2D() {desc=Velocity, parameterName=Velocity, samplerType=SAMPLERTYPE_LinearColor, texture=/Volumetrics/Content/VolumeTextures/Textures/RT_Default}
  n2 = ScalarParameter() {defaultValue=1, parameterName=TimeStep}
  n3 = Divide(B=ComponentMask(None=VectorParameter() {parameterName=Resolution, defaultValue=<rgba 1024 1024 0 0>}.RGB) {g=true, r=true})
  n4 = Subtract(A=TextureCoordinate(), B=Multiply(A=ComponentMask(None=@n1.RGB) {g=true, r=true}, B=Multiply(A=@n2, B=@n3)))
  n5 = TextureSampleParameter2D(UVs=@n4) {desc=Source, parameterName=Source, samplerType=SAMPLERTYPE_LinearColor, texture=&RT_Default}
  n6 = TextureCoordinate()
  n7 = VectorParameter() {parameterName=Resolution, defaultValue=<rgba 1024 1024 0 0>}
  n8 = Multiply(A=ScalarParameter() {defaultValue=1, parameterName=TimeStep}, B=Divide(B=ComponentMask(None=@n7.RGB) {g=true, r=true}))
  n9 = Multiply(
   A=ComponentMask(None=TextureSampleParameter2D() {desc=Velocity, parameterName=Velocity, texture=&RT_Default}.RGB) {g=true, r=true}
   B=@n8
  )
  n10 = Subtract(A=@n6, B=@n9) {desc=x n+1}
  n11 = ComponentMask(None=TextureSampleParameter2D(UVs=@n10) {desc=Fwd, parameterName=Source, texture=&RT_Default}.RGB) {g=true, r=true}
  n12 = TextureSampleParameter2D(UVs=@n10) {desc=Velocity, parameterName=Velocity, texture=&RT_Default}
  n13 = Subtract(A=@n10, B=Multiply(A=ComponentMask(None=@n12.RGB) {g=true, r=true}, B=@n8)) {desc=x n+1}
  n14 = Multiply(
   A=ComponentMask(None=TextureSampleParameter2D(UVs=@n13) {desc=Velocity, parameterName=Velocity, texture=&RT_Default}.RGB) {g=true, r=true}
   B=@n8
  )
  n15 = ComponentMask(None=TextureSampleParameter2D(UVs=Add(A=@n13, B=@n14)) {desc=Fwd, parameterName=Source, texture=&RT_Default}.RGB) {g=true, r=true}
  n16 = Multiply(A=Multiply(A=Subtract(A=@n11, B=@n15)) {constB=0.5}, B=ScalarParameter() {defaultValue=1, parameterName=MMMM})
  n17 = Constant()
  Add(A=@n5.RGB, B=Multiply(A=ScalarParameter() {parameterName=RK}, B=AppendVector(A=@n16, B=@n17)))
  Add(A=@n6, B=@n9)
  n18 = Multiply(A=@n5.B, B=ScalarParameter() {parameterName=Temp Buoyancy})
  n19 = Constant3Vector() {constant=<rgba 1 0 0 0>}
  n20 = Add(A=@n5.RGB, B=LinearInterpolate(B=Multiply(A=@n18, B=@n19), Alpha=ScalarParameter() {parameterName=IsVelocity}))
  n21 = VectorParameter() {parameterName=PaintColor, defaultValue=<rgba 0 1 0 0>}
  n22 = MaterialFunctionCall(
   UVs=TextureCoordinate()
   CenterPosition=VectorParameter() {parameterName=P, defaultValue=<rgba 0.5 0.5 0 0>}.RGB
   Radius=ScalarParameter() {defaultValue=0.05, parameterName=R}
   Density=ScalarParameter() {defaultValue=1, parameterName=D}
  ) {materialFunction=&RadialGradientExponential}
  n23 = Multiply(
   A=LinearInterpolate(A=@n22.RadialGradientExponential, Alpha=ScalarParameter() {parameterName=IsVelocity}) {constB=0}
   B=ScalarParameter() {parameterName=Color}
  )
  n24 = TextureCoordinate()
  n25 = ScalarParameter() {defaultValue=64, parameterName=CTT}
  n26 = ScalarParameter() {parameterName=ZTC}
  n27 = VectorNoise(World Position=AppendVector(
    A=Multiply(A=@n24, B=@n25)
    B=MaterialFunctionCall(Speed=@n26) {materialFunction=&TimeWithSpeedVariable}.Time
   )) {noiseFunction=VNF_CurlALU}
  n28 = Constant3Vector() {constant=<rgba 1 1 0 0>}
  n29 = Multiply(
   A=Multiply(A=Multiply(A=@n27, B=@n28), B=ScalarParameter() {defaultValue=0.1, parameterName=Curl})
   B=MaterialFunctionCall(Vector 2=VectorParameter() {parameterName=V}.RGB) {materialFunction=/Engine/Functions/Engine_MaterialFunctions02/Utility/VectorLength}.V2 Length
  )
  n30 = VectorParameter() {parameterName=V}
  n31 = Multiply(
   A=Add(A=Multiply(A=@n29, B=@n22.RadialGradientExponential), B=Multiply(A=@n22.RadialGradientExponential, B=@n30.RGB))
   B=ScalarParameter() {defaultValue=0.5, parameterName=Mouse Strength}
  )
  n32 = ScalarParameter() {parameterName=IsVelocity}
  n33 = Multiply(
   A=Add(A=LinearInterpolate(A=@n20, B=@n21.RGB, Alpha=@n23), B=Multiply(A=@n31, B=@n32))
   B=LinearInterpolate(B=ScalarParameter() {defaultValue=0.999, parameterName=Dampening}, Alpha=ScalarParameter() {parameterName=IsVelocity}) {constA=1}
  )
  ComponentMask(None=@n33) {b=true, g=true, r=true}
  Constant3Vector()
  n34 = TextureSampleParameter2D() {desc=L, parameterName=Obstacle, texture=/Engine/Functions/Engine_MaterialFunctions02/ExampleContent/Textures/SphereRenderMask}
  n35 = Divide(A=@n34.RGB, B=ScalarParameter() {defaultValue=1, desc="Default = 90 hz", parameterName=TimeStep})
  n36 = Multiply(
   A=@n33
   B=LinearInterpolate(A=LinearInterpolate(B=@n34.B) {constA=1, constAlpha=0.25}, Alpha=ScalarParameter() {parameterName=IsVelocity}) {constA=1}
  )
  LinearInterpolate(A=AppendVector(A=ComponentMask(None=@n35) {g=true, r=true}, B=Constant()), B=@n36, Alpha=@n34.B)
  LinearInterpolate(A=@n36, B=@n36, Alpha=ScalarParameter() {parameterName=IsVelocity}) {constA=1}
  n37 = TextureCoordinate()
  n38 = VectorParameter() {parameterName=Resolution, defaultValue=<rgba 1024 1024 0 0>}
  n39 = Multiply(A=ScalarParameter() {defaultValue=1, parameterName=TimeStep}, B=Divide(B=ComponentMask(None=@n38.RGB) {g=true, r=true}))
  n40 = Multiply(
   A=ComponentMask(None=TextureSampleParameter2D() {desc=Velocity, parameterName=Velocity, texture=&RT_Default}.RGB) {g=true, r=true}
   B=@n39
  )
  n41 = Add(A=@n37, B=@n40)
  n42 = Multiply(
   A=ComponentMask(None=TextureSampleParameter2D(UVs=@n41) {desc=Fwd, parameterName=Velocity, texture=&RT_Default}.RGB) {g=true, r=true}
   B=@n39
  )
  n43 = Multiply(A=Subtract(A=@n40, B=@n42)) {constB=0.5}
  n44 = ScalarParameter() {defaultValue=1, parameterName=MMMM}
  LinearInterpolate(A=@n4, B=Add(A=Subtract(A=@n37, B=@n40), B=Multiply(A=@n43, B=@n44)), Alpha=ScalarParameter() {parameterName=RK})
  OneMinus(None=ComponentMask(None=@n33) {a=true})
  n45 = TextureCoordinate() {desc=y}
  n46 = VectorParameter() {parameterName=Resolution, defaultValue=<rgba 1024 1024 0 0>}
  n47 = Multiply(A=ScalarParameter() {defaultValue=1, parameterName=TimeStep}, B=Divide(B=ComponentMask(None=@n46.RGB) {g=true, r=true}))
  n48 = Multiply(
   A=ComponentMask(None=TextureSampleParameter2D() {desc="f(yn)", parameterName=Velocity, texture=&RT_Default}.RGB) {desc=k1, g=true, r=true}
   B=Multiply(A=@n47) {constB=0.5}
  ) {desc=1/2 tk1}
  n49 = TextureSampleParameter2D(UVs=Add(A=@n45, B=@n48) {desc=yn + 1/2 tk1}) {desc="f(yn + 1/2 tk)", parameterName=Velocity, texture=&RT_Default}
  Subtract(A=TextureCoordinate(), B=Multiply(A=ComponentMask(None=@n49.RGB) {g=true, r=true}, B=@n47))
  Subtract(A=@n45, B=@n48)
  Subtract(A=TextureCoordinate(), B=Subtract(A=@n41, B=@n42))
  TextureCoordinate()
 ]
}
