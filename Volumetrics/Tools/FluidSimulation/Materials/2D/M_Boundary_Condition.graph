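# Unlit material that computes boundary conditions for a 2D fluid simulation grid, with negative emissive values allowed.
# It converts UVs to texel indices with Resolution and uses If nodes to detect the left, right, bottom and top edge texels.
# Edge texels get a summed one-texel inward UV offset, so they sample their interior neighbor, diagonally at corners.
# The sampled texture RGB is multiplied by Scale across all texels, interior included.
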Material /Volumetrics/Tools/FluidSimulation/Materials/2D/M_Boundary_Condition
graph {
 properties {bAllowNegativeEmissiveColor=true, pixelDepthOffsetMode=PDOM_Legacy, shadingModel=MSM_Unlit, usedShadingModels=Unlit}
 nodes [
  n1 = TextureCoordinate()
  n2 = TextureCoordinate()
  n3 = ScalarParameter() {defaultValue=16, parameterName=Resolution}
  n4 = MaterialFunctionCall(Float2=Ceil(None=Multiply(A=@n2, B=@n3))) {materialFunction=&BreakOutFloat2Components}
  n5 = Constant()
  n6 = Multiply(A=If(A=@n4.R, B=Constant() {r=1}, "A > B"=@n5, "A == B"=Constant() {r=1}, "A < B"=@n5) {desc=Left Boundary}, B=Constant2Vector() {r=1})
  n7 = Constant()
  n8 = Multiply(
   A=If(A=@n4.R, B=ScalarParameter() {defaultValue=16, parameterName=Resolution}, "A > B"=@n7, "A == B"=Constant() {r=1}, "A < B"=@n7) {desc=Right Boundary}
   B=Constant2Vector() {r=-1}
  )
  n9 = Constant()
  n10 = Multiply(A=If(A=@n4.G, B=Constant() {r=1}, "A > B"=@n9, "A == B"=Constant() {r=1}, "A < B"=@n9) {desc=Bottom Boundary}, B=Constant2Vector() {g=1})
  n11 = Constant()
  n12 = Multiply(
   A=If(A=@n4.G, B=ScalarParameter() {defaultValue=16, parameterName=Resolution}, "A > B"=@n11, "A == B"=Constant() {r=1}, "A < B"=@n11) {desc=Top Boundary}
   B=Constant2Vector() {g=-1}
  )
  n13 = Multiply(
   A=Add(A=Add(A=@n6, B=@n8), B=Add(A=@n10, B=@n12)) {desc=Offset in whole units}
   B=Divide(B=ScalarParameter() {defaultValue=16, parameterName=Resolution}) {desc=Texel Size}
  )
  Multiply(
   A=TextureSample(UVs=Add(A=@n1, B=@n13)) {samplerType=SAMPLERTYPE_LinearColor, texture=/Volumetrics/Content/VolumeTextures/Textures/RT_Default}.RGB
   B=ScalarParameter() {defaultValue=1, parameterName=Scale}
  )
 ]
}
