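# Material function in the fluid simulation tools that samples an input texture at the four axis neighbours of the current UV, output as -X, +X, -Y and +Y.
# The offset is one texel, derived from a Resolution vector parameter, with offset UVs clamped loosely between -1 and 2.
# A center sample, mirrored-extrapolation lerps and FluidBoundary_2D calls are present but do not reach any output.
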MaterialFunction /Volumetrics/Tools/FluidSimulation/MaterialFunctions/TextureOffsetSamples
graph {
 properties {bExposeToLibrary=true}
 nodes [
  n1 = TextureCoordinate()
  n2 = ComponentMask(None=VectorParameter() {parameterName=Resolution, defaultValue=<rgba 1024 1024 0 0>}.RGB) {g=true, r=true}
  n3 = Divide(B=@n2)
  n4 = Multiply(A=@n3, B=Constant2Vector() {r=1})
  n5 = Constant() {r=-1}
  n6 = FunctionInput() {inputName=Texture Object, inputType=FunctionInput_Texture2D}
  n7 = TextureSample(UVs=Max(A=Subtract(A=@n1, B=@n4), B=@n5), Tex=@n6)
  n8 = AppendVector(A=@n7.RGB, B=@n7.RGB)
  FunctionOutput(None=@n8) {outputName=-X}
  n9 = Constant() {r=2}
  n10 = TextureSample(UVs=Min(A=Add(A=@n1, B=@n4), B=@n9), Tex=@n6)
  n11 = AppendVector(A=@n10.RGB, B=@n10.RGB)
  FunctionOutput(None=@n11) {outputName=+X, sortPriority=1}
  n12 = TextureCoordinate()
  n13 = Multiply(A=@n3, B=Constant2Vector() {g=1})
  n14 = TextureSample(UVs=Max(A=Subtract(A=@n12, B=@n13), B=@n5), Tex=@n6)
  n15 = AppendVector(A=@n14.RGB, B=@n14.RGB)
  FunctionOutput(None=@n15) {outputName=-Y, sortPriority=2}
  n16 = TextureSample(UVs=Min(A=Add(A=@n12, B=@n13), B=@n9), Tex=@n6)
  n17 = AppendVector(A=@n16.RGB, B=@n16.RGB)
  FunctionOutput(None=@n17) {outputName=+Y, sortPriority=3}
  n18 = TextureSample(UVs=@n1, Tex=@n6) {desc=Center Sample}
  n19 = AppendVector(A=@n18.RGB, B=@n18.RGB)
  n20 = Constant()
  LinearInterpolate(A=@n8, B=Add(A=@n19, B=Subtract(A=@n19, B=@n11)), Alpha=@n20)
  LinearInterpolate(A=@n11, B=Add(A=@n19, B=Subtract(A=@n19, B=@n8)), Alpha=@n20)
  LinearInterpolate(A=@n15, B=Add(A=@n19, B=Subtract(A=@n19, B=@n17)), Alpha=@n20)
  LinearInterpolate(A=@n17, B=Add(A=@n19, B=Subtract(A=@n19, B=@n15)), Alpha=@n20)
  MaterialFunctionCall() {materialFunction=&/FluidBoundary_2D}
  MaterialFunctionCall() {materialFunction=&FluidBoundary_2D}
  Multiply(A=Divide(B=@n2)) {constB=0.5}
  OneMinus(None=@n5)
 ]
}
