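# Unlit material that computes the divergence of a 2D velocity field for the fluid simulation tools.
# TextureOffsetSamples reads Velocity and Obstacle at neighboring offsets, and Obstacle's blue channel blends in its red/green values as velocity on each side.
# Central differences in X and Y, each divided by twice GridScale, are summed into the divergence.
# Negative emissive color is allowed, permitting signed output.
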Material /Volumetrics/Tools/FluidSimulation/Materials/2D/M_Calculate_Divergence
graph {
 properties {bAllowNegativeEmissiveColor=true, pixelDepthOffsetMode=PDOM_Legacy, shadingModel=MSM_Unlit, usedShadingModels=Unlit}
 nodes [
  n1 = MaterialFunctionCall() {materialFunction=/Volumetrics/Tools/FluidSimulation/MaterialFunctions/FluidBoundary_2D}
  Add(A=OneMinus(None=Add(A=@n1.Negative X, B=@n1.Negative X)))
  Add(A=OneMinus(None=Add(A=@n1.Negative Y, B=@n1.Negative Y)))
  n2 = MaterialFunctionCall!(Texture Object=TextureObjectParameter() {parameterName=Obstacle, texture=/Engine/EngineVolumetrics/Fogsheet/Materials/T_EV_BlankWhite_01}) {materialFunction=/Volumetrics/Tools/FluidSimulation/MaterialFunctions/TextureOffsetSamples}
  n3 = MaterialFunctionCall(Texture Object=TextureObjectParameter() {parameterName=Velocity, samplerType=SAMPLERTYPE_LinearColor, texture=/Volumetrics/Content/VolumeTextures/Textures/RT_Default}) {materialFunction=&TextureOffsetSamples}
  n4 = LinearInterpolate(A=ComponentMask(None=@n2.+X) {r=true}, B=ComponentMask(None=@n3.+X) {r=true}, Alpha=ComponentMask(None=@n2.+X) {b=true})
  n5 = LinearInterpolate(A=ComponentMask(None=@n2.-X) {r=true}, B=ComponentMask(None=@n3.-X) {r=true}, Alpha=ComponentMask(None=@n2.-X) {b=true})
  n6 = ScalarParameter() {defaultValue=1, parameterName=GridScale}
  n7 = LinearInterpolate(A=ComponentMask(None=@n2.+Y) {g=true}, B=ComponentMask(None=@n3.+Y) {g=true}, Alpha=ComponentMask(None=@n2.+Y) {b=true})
  n8 = LinearInterpolate(A=ComponentMask(None=@n2.-Y) {g=true}, B=ComponentMask(None=@n3.-Y) {g=true}, Alpha=ComponentMask(None=@n2.-Y) {b=true})
  n9 = ScalarParameter() {defaultValue=1, parameterName=GridScale}
  Add(A=Divide(A=Subtract(A=@n4, B=@n5), B=Multiply(A=@n6) {constB=2}), B=Divide(A=Subtract(A=@n7, B=@n8), B=Multiply(A=@n9) {constB=2}))
  Multiply(A=Divide() {constA=0.5, desc=half rdx})
 ]
}
