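# Unlit material for the pressure gradient subtraction step of a 2D fluid simulation, producing corrected velocity.
# It takes central differences of offset Pressure samples, substituting center pressure where the Obstacle texture's blue mask marks a blocked neighbor, scales by GridScale and subtracts from Velocity.
# Per axis, neighbor masks blend toward obstacle velocity from red and green at boundaries.
# Velocity's blue channel is multiplied by obstacle blue, and negative emissive is allowed for signed values.
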Material /Volumetrics/Tools/FluidSimulation/Materials/2D/M_Gradient_Subtraction
graph {
 properties {bAllowNegativeEmissiveColor=true, pixelDepthOffsetMode=PDOM_Legacy, shadingModel=MSM_Unlit, usedShadingModels=Unlit}
 nodes [
  n1 = MaterialFunctionCall!(Texture Object=TextureObjectParameter() {parameterName=Obstacle, texture=/Engine/EngineVolumetrics/Fogsheet/Materials/T_EV_BlankWhite_01}) {materialFunction=/Volumetrics/Tools/FluidSimulation/MaterialFunctions/TextureOffsetSamples}
  n2 = ComponentMask(None=@n1.+X) {b=true}
  n3 = LinearInterpolate(A=ComponentMask(None=@n1.+X) {r=true}, B=ComponentMask(None=@n1.-X) {r=true}, Alpha=@n2) {constB=0}
  n4 = ComponentMask(None=@n1.+Y) {b=true}
  n5 = LinearInterpolate(A=ComponentMask(None=@n1.+Y) {g=true}, B=ComponentMask(None=@n1.-Y) {g=true}, Alpha=@n4) {constB=0}
  n6 = TextureSampleParameter2D(UVs=TextureCoordinate()) {
   desc=Velocity - Center Sample
   parameterName=Velocity
   samplerType=SAMPLERTYPE_LinearColor
   texture=/Volumetrics/Content/VolumeTextures/Textures/RT_Default
  }
  n7 = ComponentMask(None=@n6.RGB) {g=true, r=true}
  n8 = TextureSampleParameter2D() {desc=Center, parameterName=Pressure, samplerType=SAMPLERTYPE_LinearColor, texture=&RT_Default}
  n9 = MaterialFunctionCall(Texture Object=TextureObjectParameter() {parameterName=Pressure, texture=&RT_Default}) {materialFunction=&TextureOffsetSamples}
  n10 = ComponentMask(None=@n1.-X) {b=true}
  n11 = Subtract(
   A=LinearInterpolate(A=@n8.R, B=ComponentMask(None=@n9.+X) {r=true}, Alpha=@n2)
   B=LinearInterpolate(A=@n8.R, B=ComponentMask(None=@n9.-X) {r=true}, Alpha=@n10)
  )
  n12 = ComponentMask(None=@n1.-Y) {b=true}
  n13 = Subtract(
   A=LinearInterpolate(A=@n8.R, B=ComponentMask(None=@n9.+Y) {r=true}, Alpha=@n4)
   B=LinearInterpolate(A=@n8.R, B=ComponentMask(None=@n9.-Y) {r=true}, Alpha=@n12)
  )
  n14 = ScalarParameter() {defaultValue=1, parameterName=GridScale}
  n15 = ComponentMask(None=Divide(A=AppendVector(A=@n11, B=@n13), B=Multiply(A=@n14) {constB=2})) {g=true, r=true}
  n16 = Multiply(A=@n10, B=@n2)
  n17 = Multiply(A=@n12, B=@n4)
  AppendVector(
   A=LinearInterpolate(
    A=AppendVector(A=@n3, B=@n5) {desc=Obstacle Velocity}
    B=Subtract(A=@n7, B=@n15) {desc=Subtract Pressure Gradient}
    Alpha=AppendVector(A=@n16, B=@n17) {desc=Obstacle Mask}
   ) {desc=Kill Velocity delta towards boundaries}
   B=Multiply(A=@n6.B, B=TextureSampleParameter2D() {parameterName=Obstacle, texture=&T_EV_BlankWhite_01}.B)
  )
 ]
}
