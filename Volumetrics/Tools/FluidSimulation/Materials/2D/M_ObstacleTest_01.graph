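# Unlit test material for 2D fluid simulation obstacles, computing a three-channel field from a circular obstacle and a border box, with negative emissive values allowed.
# Red and green sum the V.xy vector and a unit direction toward P.xy scaled by V.w.
# Blue is an inverted SphereMask at P.xy with radius P.w.
# A BoxMask-3D centered in UV space, tuned by BoxBounds and BoxEdge, multiplies all three channels.
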Material /Volumetrics/Tools/FluidSimulation/Materials/2D/M_ObstacleTest_01
graph {
 properties {bAllowNegativeEmissiveColor=true, pixelDepthOffsetMode=PDOM_Legacy, shadingModel=MSM_Unlit, usedShadingModels=Unlit}
 nodes [
  n1 = ComponentMask(None=VectorParameter() {parameterName=P, defaultValue=<rgba 0.5 0.5 0 0.1>}.RGB) {g=true, r=true}
  n2 = TextureCoordinate()
  n3 = Multiply(A=Normalize(VectorInput=Subtract(A=@n1, B=@n2)), B=VectorParameter() {parameterName=V}.A)
  n4 = ComponentMask(None=VectorParameter() {parameterName=V}.RGB) {g=true, r=true}
  n5 = VectorParameter() {parameterName=P, defaultValue=<rgba 0.5 0.5 0 0.1>}
  n6 = SphereMask(A=TextureCoordinate(), B=ComponentMask(None=@n5.RGB) {g=true, r=true}, Radius=@n5.A)
  n7 = TextureCoordinate()
  n8 = Constant()
  n9 = Constant() {r=0.5}
  n10 = AppendVector(A=@n9, B=@n9)
  Multiply(
   A=AppendVector(A=Add(A=@n3, B=@n4), B=OneMinus(None=@n6))
   B=MaterialFunctionCall!(
    A=AppendVector(A=@n7, B=@n8)
    B=AppendVector(A=@n10, B=@n8)
    Bounds=ScalarParameter() {defaultValue=0.95, parameterName=BoxBounds}
    Edge Falloff=ScalarParameter() {parameterName=BoxEdge}
   ) {materialFunction=/Engine/Functions/Engine_MaterialFunctions01/Gradient/BoxMask-3D}.Result
  )
 ]
}
