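# Unlit test material for the 2D fluid simulation tools that draws a line-shaped obstacle with a velocity value.
# Three DrawLine-2D calls on 2x-tiled UVs draw a diagonal segment from P, extended by extrap, plus copies offset one unit either side in X, merged with Max.
# Halved V.xy is packed with the inverted line mask, with negative emissive allowed.
# A BoxMask-3D border mask is present but not visibly wired in.
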Material /Volumetrics/Tools/FluidSimulation/Materials/2D/M_ObstacleTest_01_line
graph {
 properties {bAllowNegativeEmissiveColor=true, pixelDepthOffsetMode=PDOM_Legacy, shadingModel=MSM_Unlit, usedShadingModels=Unlit}
 nodes [
  n1 = Constant()
  n2 = Constant() {r=0.5}
  MaterialFunctionCall!(
   A=AppendVector(A=TextureCoordinate(), B=@n1)
   B=AppendVector(A=AppendVector(A=@n2, B=@n2), B=@n1)
   Bounds=ScalarParameter() {defaultValue=0.95, parameterName=B'}
   Edge Falloff=ScalarParameter() {parameterName=E}
  ) {materialFunction=/Engine/Functions/Engine_MaterialFunctions01/Gradient/BoxMask-3D}
  n3 = ComponentMask(None=VectorParameter() {parameterName=V}.RGB) {g=true, r=true}
  n4 = TextureCoordinate() {uTiling=2, vTiling=2}
  n5 = ComponentMask(None=VectorParameter() {parameterName=P, defaultValue=<rgba 0.5 0.5 0 0.1>}.RGB) {g=true, r=true}
  n6 = ScalarParameter() {defaultValue=0.4, parameterName=extrap}
  n7 = ScalarParameter() {defaultValue=0.02, parameterName=lw}
  n8 = ComponentMask(None=VectorParameter() {parameterName=P, defaultValue=<rgba 0.5 0.5 0 0.1>}.RGB) {g=true, r=true}
  n9 = Subtract(A=@n8, B=Constant2Vector() {r=1})
  n10 = Add(A=@n9, B=@n6)
  n11 = Add(A=@n8, B=Constant2Vector() {r=1})
  n12 = Add(A=@n11, B=@n6)
  n13 = Max(
   A=MaterialFunctionCall!(Coordinates=@n4, P0=@n5, P1=Add(A=@n5, B=@n6), Line Width=@n7, Edge Falloff, Square Corners?) {materialFunction=/Engine/Functions/Engine_MaterialFunctions01/Coordinates/DrawLine-2D}.Result
   B=Max(
    A=MaterialFunctionCall(Coordinates=@n4, P0=@n9, P1=@n10, Line Width=@n7) {materialFunction=&DrawLine-2D}.Result
    B=MaterialFunctionCall(Coordinates=@n4, P0=@n11, P1=@n12, Line Width=@n7) {materialFunction=&DrawLine-2D}.Result
   )
  )
  Multiply(A=AppendVector(A=Multiply(A=@n3) {constB=0.5}, B=OneMinus(None=@n13)))
 ]
}
