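# Unlit material that runs one Jacobi-style pressure solver iteration for the 2D fluid simulation tool.
# TextureOffsetSamples fetches four Pressure and Obstacle neighbors, with Obstacle blue blending each neighbor against center pressure to handle solid boundaries.
# Their sum, minus Divergence times squared GridScale, is divided by NumSamples and multiplied by Mask blue.
# Negative emissive color is allowed, so signed pressure values survive.
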Material /Volumetrics/Tools/FluidSimulation/Materials/2D/M_Pressure_Iteration
graph {
 properties {bAllowNegativeEmissiveColor=true, pixelDepthOffsetMode=PDOM_Legacy, shadingModel=MSM_Unlit, usedShadingModels=Unlit}
 nodes [
  n1 = ScalarParameter() {defaultValue=1, parameterName=GridScale}
  n2 = TextureCoordinate()
  n3 = TextureSampleParameter2D(UVs=@n2) {desc=Center, parameterName=Divergence, samplerType=SAMPLERTYPE_LinearColor, texture=/Volumetrics/Content/VolumeTextures/Textures/RT_Default}
  n4 = Multiply(A=Multiply(A=Multiply(A=@n1, B=@n1)) {constB=-1}, B=@n3.RGB)
  n5 = TextureSampleParameter2D(UVs=@n2) {desc=Center, parameterName=Pressure, samplerType=SAMPLERTYPE_LinearColor, texture=&RT_Default}
  n6 = MaterialFunctionCall!(Texture Object=TextureObjectParameter() {parameterName=Pressure, texture=&RT_Default}) {materialFunction=/Volumetrics/Tools/FluidSimulation/MaterialFunctions/TextureOffsetSamples}
  n7 = ComponentMask(None=@n6.-X) {r=true}
  n8 = MaterialFunctionCall(Texture Object=TextureObjectParameter() {parameterName=Obstacle, texture=/Engine/EngineVolumetrics/Fogsheet/Materials/T_EV_BlankWhite_01}) {materialFunction=&TextureOffsetSamples}
  n9 = ComponentMask(None=@n8.-X) {b=true}
  n10 = ComponentMask(None=@n6.+X) {r=true}
  n11 = ComponentMask(None=@n8.+X) {b=true}
  n12 = ComponentMask(None=@n6.-Y) {r=true}
  n13 = ComponentMask(None=@n8.-Y) {b=true}
  n14 = ComponentMask(None=@n6.+Y) {r=true}
  n15 = ComponentMask(None=@n8.+Y) {b=true}
  n16 = Add(
   A=Add(A=LinearInterpolate(A=@n5.RGB, B=@n7, Alpha=@n9), B=LinearInterpolate(A=@n5.RGB, B=@n10, Alpha=@n11))
   B=Add(A=LinearInterpolate(A=@n5.RGB, B=@n12, Alpha=@n13), B=LinearInterpolate(A=@n5.RGB, B=@n14, Alpha=@n15))
  )
  n17 = ScalarParameter() {defaultValue=4, parameterName=NumSamples}
  Multiply(
   A=TextureSampleParameter2D() {desc=L, parameterName=Mask, texture=&T_EV_BlankWhite_01}.B
   B=Multiply(A=Add(A=@n4, B=@n16), B=Divide(B=@n17) {desc=Reciprocal})
  )
 ]
}
