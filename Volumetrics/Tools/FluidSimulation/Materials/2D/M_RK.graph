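# Unlit material for the 2D fluid simulation tools that computes an advection error correction on a Source texture using a Velocity texture.
# It traces UVs backward by velocity scaled by TimeStep over Resolution, resamples velocity there, then traces forward to resample Source.
# Half the red-green difference between original and round-trip Source, scaled by MMMM, is added back to Source, resembling BFECC/MacCormack correction despite the RK name.
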Material /Volumetrics/Tools/FluidSimulation/Materials/2D/M_RK
graph {
 properties {bAllowNegativeEmissiveColor=true, pixelDepthOffsetMode=PDOM_Legacy, shadingModel=MSM_Unlit, usedShadingModels=Unlit}
 nodes [
  n1 = TextureSampleParameter2D() {desc=Fwd, parameterName=Source, samplerType=SAMPLERTYPE_LinearColor, texture=/Volumetrics/Content/VolumeTextures/Textures/RT_Default}
  n2 = TextureSampleParameter2D() {desc=Velocity, parameterName=Velocity, samplerType=SAMPLERTYPE_LinearColor, texture=&RT_Default}
  n3 = VectorParameter() {parameterName=Resolution, defaultValue=<rgba 1024 1024 0 0>}
  n4 = Multiply(A=ScalarParameter() {defaultValue=1, parameterName=TimeStep}, B=Divide(B=ComponentMask(None=@n3.RGB) {g=true, r=true}))
  n5 = Subtract(A=TextureCoordinate(), B=Multiply(A=ComponentMask(None=@n2.RGB) {g=true, r=true}, B=@n4)) {desc=x n+1}
  n6 = TextureSampleParameter2D(UVs=@n5) {
   desc="Velocity @ N + 1"
   parameterName=Velocity
   samplerType=SAMPLERTYPE_LinearColor
   texture=&RT_Default
  }
  n7 = ComponentMask(None=@n6.RGB) {g=true, r=true}
  n8 = TextureSampleParameter2D(UVs=Add(A=@n5, B=Multiply(A=@n7, B=@n4))) {desc=Fwd, parameterName=Source, samplerType=SAMPLERTYPE_LinearColor, texture=&RT_Default}
  n9 = Multiply(A=Subtract(A=ComponentMask(None=@n1.RGB) {g=true, r=true}, B=ComponentMask(None=@n8.RGB) {g=true, r=true})) {constB=0.5}
  n10 = ScalarParameter() {defaultValue=1, parameterName=MMMM}
  Add(A=@n1.RGB, B=AppendVector(A=Multiply(A=@n9, B=@n10), B=Constant()))
 ]
}
