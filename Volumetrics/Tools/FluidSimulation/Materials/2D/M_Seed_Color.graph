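# M_Seed_Color is an unlit material in the 2D fluid simulation tools, named as a color seed for the simulation.
# Its graph holds a single texture sample of the engine's debug HeatmapGradient texture, with no custom UVs, parameters or math.
# The graph shows no link from that sample to a material output.
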Material /Volumetrics/Tools/FluidSimulation/Materials/2D/M_Seed_Color
graph {
 properties {pixelDepthOffsetMode=PDOM_Legacy, shadingModel=MSM_Unlit, usedShadingModels=Unlit}
 nodes [TextureSample() {texture=/Engine/EngineDebugMaterials/HeatmapGradient}]
}
