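# Unlit material from the 2D fluid simulation tools that, by its name, seeds velocity around a point in UV space.
# A sphere mask at SpeedPoint shapes curl vector noise and a radial force, which are summed with a constant velocity into an XY vector.
# A separate branch appends a masked temperature seed to scaled XY values from a sampled texture.
# The graph shows no material output connection and holds several stray constant, lerp and error-texture nodes.
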Material /Volumetrics/Tools/FluidSimulation/Materials/2D/M_Seed_Velocity
graph {
 properties {bAllowNegativeEmissiveColor=true, pixelDepthOffsetMode=PDOM_Legacy, shadingModel=MSM_Unlit, usedShadingModels=Unlit}
 nodes [
  n1 = Multiply(A=AppendVector(A=TextureCoordinate(), B=ScalarParameter() {parameterName=Z}), B=ScalarParameter() {defaultValue=5, parameterName=PT})
  n2 = TextureCoordinate()
  n3 = ComponentMask(None=VectorParameter() {parameterName=SpeedPoint, defaultValue=<rgba 0.2 0.5 0 0>}.RGB) {g=true, r=true}
  n4 = SphereMask(A=@n2, B=@n3, Radius=ScalarParameter() {defaultValue=0.1, parameterName=R}, Hardness=ScalarParameter() {defaultValue=0.25, parameterName=H})
  n5 = Multiply(A=Multiply(A=VectorNoise(World Position=@n1) {noiseFunction=VNF_CurlALU}, B=ScalarParameter() {parameterName=E}), B=@n4)
  n6 = Multiply(A=Subtract(A=@n2, B=@n3), B=@n4)
  n7 = ScalarParameter() {parameterName=Radial Force}
  Add(
   A=Add(A=ComponentMask(None=@n5) {g=true, r=true}, B=Multiply(A=@n6, B=@n7)) {constB=0}
   B=ComponentMask(None=VectorParameter() {parameterName=Constant Velocity}.RGB) {g=true, r=true}
  )
  n8 = TextureSample() {samplerType=SAMPLERTYPE_LinearColor, texture=/Volumetrics/Content/VolumeTextures/Textures/outputtest}
  n9 = ScalarParameter() {defaultValue=0.02, parameterName=Vel}
  AppendVector(A=ComponentMask(None=Multiply(A=@n8.RGB, B=@n9)) {g=true, r=true}, B=Multiply(A=@n4, B=ScalarParameter() {parameterName=Temp Seed}))
  Constant() {r=0.5}
  Constant()
  LinearInterpolate(
   A=Constant3Vector() {constant=<rgba 0 0 1 0>}
   B=Constant3Vector() {constant=<rgba 0 1 0 0>}
   Alpha=Constant3Vector() {constant=<rgba 1 1 0 0>}
  )
  TextureSample() {texture=/Engine/EngineDebugMaterials/MaterialError}
 ]
}
